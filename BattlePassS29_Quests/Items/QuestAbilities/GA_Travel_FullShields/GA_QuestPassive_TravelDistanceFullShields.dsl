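; Passive quest ability that tags the owning player as having full or not-full shields via gameplay effects, apparently for a travel-distance-with-full-shields quest.
; On activation, a looping timer with a randomized interval runs the misnamed Check Ammo event, applying FullShieldsGE only when both shield and overshield are at maximum, otherwise NotFullShieldsGE.
; A damage-received event listener also applies NotFullShieldsGE immediately, between polls.
; Ending the ability clears the timer.
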(event Ability|EventFailedToActivatePassiveAbility
  (params (ActorInfo :type Gameplay%20Ability%20Actor%20Info%20Structure))
  (|CallParentFunction ActorInfo))

(event Ability|EventOnEndAbility
  (params (bWasCancelled :type Boolean))
  (|CallParentFunction bWasCancelled)
  (Utilities|Time|ClearandInvalidateTimerbyHandle (Variables|Default|GetTimer%20Handle)))

(event Ability|EventActivateAbility
  (|CallParentFunction)
  (bind _returnvalue (Utilities|Time|SetTimerbyEvent (delegate Custom|Check%20Ammo) (Math|Random|RandomFloatinRange 1.0 1.5) true))
  (Variables|Default|SetTimer%20Handle _returnvalue)
  (Ability|Tasks|WaitGameplayEvent "(TagName=\"Event.Damage.Received\")"
    (:EventReceived
      (GameplayEffects|ApplyGameplayEffectToSelf :self (Ability|GetFortAbilitySystemComponent (Variables|GAVars|Data|GetOwningPlayerPawn)) :GameplayEffectClass (Variables|Default|GetNotFullShieldsGE) :Level 0.0))))

(event Custom|Check%20Ammo
  (bind _owningplayerpawn (Variables|GAVars|Data|GetOwningPlayerPawn))
  (bind _returnvalue (Ability|GetFortAbilitySystemComponent _owningplayerpawn))
  ;[1 1 1] Check if player has/has not full shields, and apply GEs
  (if (and (>= (FortPawn|GetShield (Variables|GAVars|Data|GetOwningPlayerPawn)) (FortPawn|GetMaxShield (Variables|GAVars|Data|GetOwningPlayerPawn))) (>= (FortPawn|GetOvershield (Variables|GAVars|Data|GetOwningPlayerPawn)) (FortPawn|GetMaxOvershield (Variables|GAVars|Data|GetOwningPlayerPawn))))
    (GameplayEffects|ApplyGameplayEffectToSelf :self _returnvalue :GameplayEffectClass (Variables|Default|GetFullShieldsGE) :Level 0.0)
    (else
      (GameplayEffects|ApplyGameplayEffectToSelf :self _returnvalue :GameplayEffectClass (Variables|Default|GetNotFullShieldsGE) :Level 0.0))))
